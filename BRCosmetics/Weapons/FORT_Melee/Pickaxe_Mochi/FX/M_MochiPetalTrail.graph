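# Unlit, translucent, two-sided Niagara ribbon material for the Mochi pickaxe's petal trail.
# A petal texture panning on UV1 supplies a mask, faded at both ends along UV0, that helps blend dark and light pink colors.
# Particle color tints the result before an emissive adjustment function.
# Opacity combines the mask, view-angle fade, particle alpha and depth fade, which mobile replaces with a constant.
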Material /BRCosmetics/Weapons/FORT_Melee/Pickaxe_Mochi/FX/M_MochiPetalTrail
graph {
 properties {
  bUsedWithNiagaraRibbons=true
  blendMode=BLEND_Translucent
  pixelDepthOffsetMode=PDOM_Legacy
  shadingModel=MSM_Unlit
  twoSided=true
  usedShadingModels=Unlit
 }
 nodes [
  n1 = TextureSample(UVs=Panner(Coordinate=TextureCoordinate() {coordinateIndex=1}) {speedX=-1}) {samplerType=SAMPLERTYPE_LinearColor, texture=&/T_MochiPetals_M}
  n2 = ComponentMask(None=TextureCoordinate()) {r=true}
  n3 = Multiply(A=@n1.R, B=Multiply(A=OneMinus(None=@n2), B=SmoothStep(Value=@n2) {constMax=0.1}))
  n4 = LinearInterpolate(
   A=VectorParameter() {parameterName=DarkColor, defaultValue=<rgba 0.2196 0.051 0.1373 1>}.RGB
   B=VectorParameter() {parameterName=LightColor, defaultValue=<rgba 0.4275 0.102 0.2706 1>}.RGB
   Alpha=Multiply(A=Multiply(A=@n1.G) {constB=4}, B=Multiply(A=@n3) {constB=4})
  )
  n5 = ParticleColor()
  MaterialFunctionCall(Emissive=Multiply(A=Saturate(None=@n4), B=@n5.RGB)) {materialFunction=&MF_FX_EmissiveAdjustments}
  n6 = CameraVectorWS()
  n7 = VertexNormalWS()
  n8 = Multiply(A=SmoothStep(Value=@n3) {constMax=0.6}, B=SmoothStep(Value=DotProduct(A=@n6, B=@n7)) {constMax=0.5})
  n9 = DepthFade() {fadeDistanceDefault=25}
  n10 = ShadingPathSwitch(Default=@n9, Deferred=@n9, Mobile=Constant() {r=1}) {desc=Migrated from FeatureLevelSwitch}
  Multiply(A=Saturate(None=Multiply(A=@n8, B=@n10)), B=@n5.A)
 ]
}
