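# Physics asset for the Mochi pickaxe's skeletal mesh, letting 27 small mochi pieces wobble on the weapon.
# Each dyn_mochi capsule hinges on its base_mochi anchor with almost no twist and either even or one-axis-biased swing.
# Anchors 01–24 connect through a shared dyn_base to the handle, while 25–27 attach directly to the handle.
# Those anchor joints allow slight swing; the handle-to-root joint is free in rotation.
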PhysicsAsset /BRCosmetics/Weapons/FORT_Melee/Pickaxe_Mochi/Meshes/Mochi_Axe_Physics
bodies {
 Melee_TwoHanded_Handle [{shapeType=Capsule, length=10, radius=2, shapeName=Melee_TwoHanded_Handle_capsule}]
 b_AR_Root [{shapeType=Capsule, length=2, radius=2, shapeName=b_AR_Root_capsule, rotation=<rot 10.0001 0 0>}]
 base_mochi01 [{shapeType=Capsule, length=0.505, radius=0.505, shapeName=base_mochi01_capsule, rotation=<rot 90 0 0>}]
 base_mochi02 [{shapeType=Capsule, length=0.505, radius=0.505, shapeName=base_mochi02_capsule, rotation=<rot 90 0 0>}]
 base_mochi03 [{shapeType=Capsule, length=0.505, radius=0.505, shapeName=base_mochi03_capsule, rotation=<rot 90 0 0>}]
 base_mochi04 [{shapeType=Capsule, length=0.505, radius=0.505, shapeName=base_mochi04_capsule, rotation=<rot 90 0 0>}]
 base_mochi05 [{shapeType=Capsule, length=0.505, radius=0.505, shapeName=base_mochi05_capsule, rotation=<rot 90 0 0>}]
 base_mochi06 [{shapeType=Capsule, length=0.505, radius=0.505, shapeName=base_mochi06_capsule, rotation=<rot 90 0 0>}]
 base_mochi07 [{shapeType=Capsule, length=0.505, radius=0.505, shapeName=base_mochi07_capsule, rotation=<rot 90 0 0>}]
 base_mochi08 [{shapeType=Capsule, length=0.505, radius=0.505, shapeName=base_mochi08_capsule, rotation=<rot 90 0 0>}]
 base_mochi09 [{shapeType=Capsule, length=0.505, radius=0.505, shapeName=base_mochi09_capsule, rotation=<rot 90 0 0>}]
 base_mochi10 [{shapeType=Capsule, length=0.505, radius=0.505, shapeName=base_mochi10_capsule, rotation=<rot 90 0 0>}]
 base_mochi11 [{shapeType=Capsule, length=0.505, radius=0.505, shapeName=base_mochi11_capsule, rotation=<rot 90 0 0>}]
 base_mochi12 [{shapeType=Capsule, length=0.505, radius=0.505, shapeName=base_mochi12_capsule, rotation=<rot 90 0 0>}]
 base_mochi13 [{shapeType=Capsule, length=0.505, radius=0.505, shapeName=base_mochi13_capsule, rotation=<rot 90 0 0>}]
 base_mochi14 [{shapeType=Capsule, length=0.505, radius=0.505, shapeName=base_mochi14_capsule, rotation=<rot 90 0 0>}]
 base_mochi15 [{shapeType=Capsule, length=0.505, radius=0.505, shapeName=base_mochi15_capsule, rotation=<rot 90 0 0>}]
 base_mochi16 [{shapeType=Capsule, length=0.505, radius=0.505, shapeName=base_mochi16_capsule, rotation=<rot 90 0 0>}]
 base_mochi17 [{shapeType=Capsule, length=0.505, radius=0.505, shapeName=base_mochi17_capsule, rotation=<rot 90 0 0>}]
 base_mochi18 [{shapeType=Capsule, length=0.505, radius=0.505, shapeName=base_mochi18_capsule, rotation=<rot 90 0 0>}]
 base_mochi19 [{shapeType=Capsule, length=0.505, radius=0.505, shapeName=base_mochi19_capsule, rotation=<rot 90 0 0>}]
 base_mochi20 [{shapeType=Capsule, length=0.505, radius=0.505, shapeName=base_mochi20_capsule, rotation=<rot 90 0 0>}]
 base_mochi21 [{shapeType=Capsule, length=0.505, radius=0.505, shapeName=base_mochi21_capsule, rotation=<rot 90 0 0>}]
 base_mochi22 [{shapeType=Capsule, length=0.505, radius=0.505, shapeName=base_mochi22_capsule, rotation=<rot 90 0 0>}]
 base_mochi23 [{shapeType=Capsule, length=0.505, radius=0.505, shapeName=base_mochi23_capsule, rotation=<rot 90 0 0>}]
 base_mochi24 [{shapeType=Capsule, length=0.505, radius=0.505, shapeName=base_mochi24_capsule, rotation=<rot 90 0 0>}]
 base_mochi25 [{shapeType=Capsule, length=0.505, radius=0.505, shapeName=base_mochi25_capsule, rotation=<rot 90 0 0>}]
 base_mochi26 [{shapeType=Capsule, length=0.505, radius=0.505, shapeName=base_mochi26_capsule, rotation=<rot 90 0 0>}]
 base_mochi27 [{shapeType=Capsule, length=0.505, radius=0.505, shapeName=base_mochi27_capsule, rotation=<rot 90 0 0>}]
 dyn_base [{shapeType=Capsule, length=0.505, radius=0.505, shapeName=dyn_base_capsule, rotation=<rot 90 0 0>}]
 dyn_mochi01 [{shapeType=Capsule, length=2, radius=2, shapeName=dyn_mochi01_capsule, center=<4 0 0>, rotation=<rot 90 136.4627 136.4627>}]
 dyn_mochi02 [{shapeType=Capsule, length=2, radius=2, shapeName=dyn_mochi02_capsule, center=<4 0 0>, rotation=<rot 90 -104.1513 -104.1513>}]
 dyn_mochi03 [{shapeType=Capsule, length=2, radius=2, shapeName=dyn_mochi03_capsule, center=<4 0 0>, rotation=<rot 90 53.731 53.731>}]
 dyn_mochi04 [{shapeType=Capsule, length=2, radius=2, shapeName=dyn_mochi04_capsule, center=<4 0 0>, rotation=<rot 90 -133.6167 -133.6167>}]
 dyn_mochi05 [{shapeType=Capsule, length=2, radius=2, shapeName=dyn_mochi05_capsule, center=<4 0 0>, rotation=<rot 90 0 0>}]
 dyn_mochi06 [{shapeType=Capsule, length=2, radius=2, shapeName=dyn_mochi06_capsule, center=<4 0 0>, rotation=<rot 90 -38.7129 -38.7129>}]
 dyn_mochi07 [{shapeType=Capsule, length=2, radius=2, shapeName=dyn_mochi07_capsule, center=<4 0 0>, rotation=<rot 90 119.4293 119.4293>}]
 dyn_mochi08 [{shapeType=Capsule, length=2, radius=2, shapeName=dyn_mochi08_capsule, center=<4 0 0>, rotation=<rot 90 161.5708 161.5708>}]
 dyn_mochi09 [{shapeType=Capsule, length=2, radius=2, shapeName=dyn_mochi09_capsule, center=<4 0 0>, rotation=<rot 90 124.8892 124.8892>}]
 dyn_mochi10 [{shapeType=Capsule, length=2, radius=2, shapeName=dyn_mochi10_capsule, center=<4 0 0>, rotation=<rot 90 125.1084 125.1084>}]
 dyn_mochi11 [{shapeType=Capsule, length=2, radius=2, shapeName=dyn_mochi11_capsule, center=<4 0 0>, rotation=<rot 90 95.2939 95.2939>}]
 dyn_mochi12 [{shapeType=Capsule, length=2, radius=2, shapeName=dyn_mochi12_capsule, center=<4 0 0>, rotation=<rot 90 -66.9243 -66.9243>}]
 dyn_mochi13 [{shapeType=Capsule, length=2, radius=2, shapeName=dyn_mochi13_capsule, center=<4 0 0>, rotation=<rot 90 -48.1641 -48.1641>}]
 dyn_mochi14 [{shapeType=Capsule, length=2, radius=2, shapeName=dyn_mochi14_capsule, center=<4 0 0>, rotation=<rot 90 -26.699 -26.699>}]
 dyn_mochi15 [{shapeType=Capsule, length=2, radius=2, shapeName=dyn_mochi15_capsule, center=<4 0 0>, rotation=<rot 90 116.9039 116.9039>}]
 dyn_mochi16 [{shapeType=Capsule, length=2, radius=2, shapeName=dyn_mochi16_capsule, center=<4 0 0>, rotation=<rot 90 -26.6991 -26.6991>}]
 dyn_mochi17 [{shapeType=Capsule, length=2, radius=2, shapeName=dyn_mochi17_capsule, center=<4 0 0>, rotation=<rot 90 -48.1641 -48.1642>}]
 dyn_mochi18 [{shapeType=Capsule, length=2, radius=2, shapeName=dyn_mochi18_capsule, center=<4 0 0>, rotation=<rot 90 -66.9244 -66.9244>}]
 dyn_mochi19 [{shapeType=Capsule, length=2, radius=2, shapeName=dyn_mochi19_capsule, center=<4 0 0>, rotation=<rot 90 -54.8916 -54.8916>}]
 dyn_mochi20 [{shapeType=Capsule, length=2, radius=2, shapeName=dyn_mochi20_capsule, center=<4 0 0>, rotation=<rot 90 -84.7061 -84.7061>}]
 dyn_mochi21 [{shapeType=Capsule, length=2, radius=2, shapeName=dyn_mochi21_capsule, center=<4 0 0>, rotation=<rot 90 124.8899 124.8899>}]
 dyn_mochi22 [{shapeType=Capsule, length=2, radius=2, shapeName=dyn_mochi22_capsule, center=<4 0 0>, rotation=<rot 90 161.5708 161.5708>}]
 dyn_mochi23 [{shapeType=Capsule, length=2, radius=2, shapeName=dyn_mochi23_capsule, center=<4 0 0>, rotation=<rot 90 119.4293 119.4293>}]
 dyn_mochi24 [{shapeType=Capsule, length=2, radius=2, shapeName=dyn_mochi24_capsule, center=<4 0 0>, rotation=<rot 90 141.287 141.287>}]
 dyn_mochi25 [{shapeType=Capsule, length=2, radius=2, shapeName=dyn_mochi25_capsule, center=<4 0 0>, rotation=<rot 90 -35.4012 -35.4012>}]
 dyn_mochi26 [{shapeType=Capsule, length=2, radius=2, shapeName=dyn_mochi26_capsule, center=<4 0 0>, rotation=<rot 90 -109.4919 -109.4919>}]
 dyn_mochi27 [{shapeType=Capsule, length=2, radius=2, shapeName=dyn_mochi27_capsule, center=<4 0 0>, rotation=<rot 90 93.7438 93.7438>}]
}
constraints [
 {bone1Name=dyn_mochi01, bone2Name=base_mochi01, swing1LimitDegrees=30, swing2LimitDegrees=30, twistLimitDegrees=1}
 {bone1Name=dyn_mochi02, bone2Name=base_mochi02, swing1LimitDegrees=30, swing2LimitDegrees=30, twistLimitDegrees=1}
 {bone1Name=dyn_mochi03, bone2Name=base_mochi03, swing1LimitDegrees=30, swing2LimitDegrees=30, twistLimitDegrees=1}
 {bone1Name=dyn_mochi04, bone2Name=base_mochi04, swing1LimitDegrees=30, swing2LimitDegrees=30, twistLimitDegrees=1}
 {bone1Name=dyn_mochi05, bone2Name=base_mochi05, swing1LimitDegrees=50, swing2LimitDegrees=10, twistLimitDegrees=1}
 {bone1Name=dyn_mochi06, bone2Name=base_mochi06, swing1LimitDegrees=45, swing2LimitDegrees=10, twistLimitDegrees=1}
 {bone1Name=dyn_mochi07, bone2Name=base_mochi07, swing1LimitDegrees=45, swing2LimitDegrees=10, twistLimitDegrees=1}
 {bone1Name=dyn_mochi08, bone2Name=base_mochi08, swing1LimitDegrees=45, swing2LimitDegrees=10, twistLimitDegrees=1}
 {bone1Name=dyn_mochi09, bone2Name=base_mochi09, swing1LimitDegrees=45, swing2LimitDegrees=10, twistLimitDegrees=1}
 {bone1Name=dyn_mochi10, bone2Name=base_mochi10, swing1LimitDegrees=45, swing2LimitDegrees=10, twistLimitDegrees=1}
 {bone1Name=dyn_mochi11, bone2Name=base_mochi11, swing1LimitDegrees=45, swing2LimitDegrees=10, twistLimitDegrees=1}
 {bone1Name=dyn_mochi12, bone2Name=base_mochi12, swing1LimitDegrees=45, swing2LimitDegrees=10, twistLimitDegrees=1}
 {bone1Name=dyn_mochi13, bone2Name=base_mochi13, swing1LimitDegrees=45, swing2LimitDegrees=10, twistLimitDegrees=1}
 {bone1Name=dyn_mochi14, bone2Name=base_mochi14, swing1LimitDegrees=45, swing2LimitDegrees=10, twistLimitDegrees=1}
 {bone1Name=dyn_mochi15, bone2Name=base_mochi15, swing1LimitDegrees=50, swing2LimitDegrees=10, twistLimitDegrees=1}
 {bone1Name=dyn_mochi16, bone2Name=base_mochi16, swing1LimitDegrees=45, swing2LimitDegrees=10, twistLimitDegrees=1}
 {bone1Name=dyn_mochi17, bone2Name=base_mochi17, swing1LimitDegrees=45, swing2LimitDegrees=10, twistLimitDegrees=1}
 {bone1Name=dyn_mochi18, bone2Name=base_mochi18, swing1LimitDegrees=50, swing2LimitDegrees=10, twistLimitDegrees=1}
 {bone1Name=dyn_mochi19, bone2Name=base_mochi19, swing1LimitDegrees=45, swing2LimitDegrees=10, twistLimitDegrees=1}
 {bone1Name=dyn_mochi20, bone2Name=base_mochi20, swing1LimitDegrees=45, swing2LimitDegrees=10, twistLimitDegrees=1}
 {bone1Name=dyn_mochi21, bone2Name=base_mochi21, swing1LimitDegrees=45, swing2LimitDegrees=10, twistLimitDegrees=1}
 {bone1Name=dyn_mochi22, bone2Name=base_mochi22, swing1LimitDegrees=45, swing2LimitDegrees=10, twistLimitDegrees=1}
 {bone1Name=dyn_mochi23, bone2Name=base_mochi23, swing1LimitDegrees=45, swing2LimitDegrees=10, twistLimitDegrees=1}
 {bone1Name=dyn_mochi24, bone2Name=base_mochi24, swing1LimitDegrees=45, swing2LimitDegrees=10, twistLimitDegrees=1}
 {bone1Name=dyn_mochi25, bone2Name=base_mochi25, swing1LimitDegrees=30, swing2LimitDegrees=30, twistLimitDegrees=1}
 {bone1Name=dyn_mochi26, bone2Name=base_mochi26, swing1LimitDegrees=30, swing2LimitDegrees=30, twistLimitDegrees=1}
 {bone1Name=dyn_mochi27, bone2Name=base_mochi27, swing1LimitDegrees=30, swing2LimitDegrees=30, twistLimitDegrees=1}
 {bone1Name=dyn_base, bone2Name=Melee_TwoHanded_Handle, swing1LimitDegrees=10, swing2LimitDegrees=10, twistLimitDegrees=1}
 {bone1Name=base_mochi01, bone2Name=dyn_base, swing1LimitDegrees=10, swing2LimitDegrees=10, twistLimitDegrees=1}
 {bone1Name=base_mochi02, bone2Name=dyn_base, swing1LimitDegrees=10, swing2LimitDegrees=10, twistLimitDegrees=1}
 {bone1Name=base_mochi03, bone2Name=dyn_base, swing1LimitDegrees=10, swing2LimitDegrees=10, twistLimitDegrees=1}
 {bone1Name=base_mochi04, bone2Name=dyn_base, swing1LimitDegrees=10, swing2LimitDegrees=10, twistLimitDegrees=1}
 {bone1Name=base_mochi05, bone2Name=dyn_base, swing1LimitDegrees=10, swing2LimitDegrees=10, twistLimitDegrees=1}
 {bone1Name=base_mochi06, bone2Name=dyn_base, swing1LimitDegrees=10, swing2LimitDegrees=10, twistLimitDegrees=1}
 {bone1Name=base_mochi07, bone2Name=dyn_base, swing1LimitDegrees=10, swing2LimitDegrees=10, twistLimitDegrees=1}
 {bone1Name=base_mochi08, bone2Name=dyn_base, swing1LimitDegrees=10, swing2LimitDegrees=10, twistLimitDegrees=1}
 {bone1Name=base_mochi09, bone2Name=dyn_base, swing1LimitDegrees=10, swing2LimitDegrees=10, twistLimitDegrees=1}
 {bone1Name=base_mochi10, bone2Name=dyn_base, swing1LimitDegrees=10, swing2LimitDegrees=10, twistLimitDegrees=1}
 {bone1Name=base_mochi11, bone2Name=dyn_base, swing1LimitDegrees=10, swing2LimitDegrees=10, twistLimitDegrees=1}
 {bone1Name=base_mochi12, bone2Name=dyn_base, swing1LimitDegrees=10, swing2LimitDegrees=10, twistLimitDegrees=1}
 {bone1Name=base_mochi13, bone2Name=dyn_base, swing1LimitDegrees=10, swing2LimitDegrees=10, twistLimitDegrees=1}
 {bone1Name=base_mochi14, bone2Name=dyn_base, swing1LimitDegrees=10, swing2LimitDegrees=10, twistLimitDegrees=1}
 {bone1Name=base_mochi15, bone2Name=dyn_base, swing1LimitDegrees=10, swing2LimitDegrees=10, twistLimitDegrees=1}
 {bone1Name=base_mochi16, bone2Name=dyn_base, swing1LimitDegrees=10, swing2LimitDegrees=10, twistLimitDegrees=1}
 {bone1Name=base_mochi17, bone2Name=dyn_base, swing1LimitDegrees=10, swing2LimitDegrees=10, twistLimitDegrees=1}
 {bone1Name=base_mochi18, bone2Name=dyn_base, swing1LimitDegrees=10, swing2LimitDegrees=10, twistLimitDegrees=1}
 {bone1Name=base_mochi19, bone2Name=dyn_base, swing1LimitDegrees=10, swing2LimitDegrees=10, twistLimitDegrees=1}
 {bone1Name=base_mochi20, bone2Name=dyn_base, swing1LimitDegrees=10, swing2LimitDegrees=10, twistLimitDegrees=1}
 {bone1Name=base_mochi21, bone2Name=dyn_base, swing1LimitDegrees=10, swing2LimitDegrees=10, twistLimitDegrees=1}
 {bone1Name=base_mochi22, bone2Name=dyn_base, swing1LimitDegrees=10, swing2LimitDegrees=10, twistLimitDegrees=1}
 {bone1Name=base_mochi23, bone2Name=dyn_base, swing1LimitDegrees=10, swing2LimitDegrees=10, twistLimitDegrees=1}
 {bone1Name=base_mochi24, bone2Name=dyn_base, swing1LimitDegrees=10, swing2LimitDegrees=10, twistLimitDegrees=1}
 {bone1Name=base_mochi26, bone2Name=Melee_TwoHanded_Handle, swing1LimitDegrees=10, swing2LimitDegrees=10, twistLimitDegrees=1}
 {bone1Name=base_mochi25, bone2Name=Melee_TwoHanded_Handle, swing1LimitDegrees=10, swing2LimitDegrees=10, twistLimitDegrees=1}
 {bone1Name=base_mochi27, bone2Name=Melee_TwoHanded_Handle, swing1LimitDegrees=10, swing2LimitDegrees=10, twistLimitDegrees=1}
 {
  bone1Name=Melee_TwoHanded_Handle
  bone2Name=b_AR_Root
  swing1LimitDegrees=45
  swing1Motion=Free
  swing2LimitDegrees=45
  swing2Motion=Free
  twistLimitDegrees=45
  twistMotion=Free
 }
]
