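# Physics asset for the Mochi pickaxe leaves mesh, giving its dynamic bones capsule bodies and joints for secondary motion.
# Seven dyn_mochi capsules, offset from their bones, each pair with a small base_mochi anchor through wider swing limits and very tight twist.
# Four anchors chain through dyn_base to the two-handed handle and three attach to it directly, all with narrower swing.
# The handle joins the root with free swing and twist.
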PhysicsAsset /BRCosmetics/Weapons/FORT_Melee/Pickaxe_Mochi/Meshes/Mochi_Axe_leaves_Physics
bodies {
 Melee_TwoHanded_Handle [{shapeType=Capsule, length=10, radius=2, shapeName=Melee_TwoHanded_Handle_capsule}]
 b_AR_Root [{shapeType=Capsule, length=2, radius=2, shapeName=b_AR_Root_capsule, rotation=<rot 10.0001 0 0>}]
 base_mochi01 [{shapeType=Capsule, length=0.505, radius=0.505, shapeName=base_mochi01_capsule, rotation=<rot 90 0 0>}]
 base_mochi02 [{shapeType=Capsule, length=0.505, radius=0.505, shapeName=base_mochi02_capsule, rotation=<rot 90 0 0>}]
 base_mochi03 [{shapeType=Capsule, length=0.505, radius=0.505, shapeName=base_mochi03_capsule, rotation=<rot 90 0 0>}]
 base_mochi04 [{shapeType=Capsule, length=0.505, radius=0.505, shapeName=base_mochi04_capsule, rotation=<rot 90 0 0>}]
 base_mochi25 [{shapeType=Capsule, length=0.505, radius=0.505, shapeName=base_mochi25_capsule, rotation=<rot 90 0 0>}]
 base_mochi26 [{shapeType=Capsule, length=0.505, radius=0.505, shapeName=base_mochi26_capsule, rotation=<rot 90 0 0>}]
 base_mochi27 [{shapeType=Capsule, length=0.505, radius=0.505, shapeName=base_mochi27_capsule, rotation=<rot 90 0 0>}]
 dyn_base [{shapeType=Capsule, length=0.505, radius=0.505, shapeName=dyn_base_capsule, rotation=<rot 90 0 0>}]
 dyn_mochi01 [{shapeType=Capsule, length=2, radius=2, shapeName=dyn_mochi01_capsule, center=<4 0 0>, rotation=<rot 90 136.4627 136.4627>}]
 dyn_mochi02 [{shapeType=Capsule, length=2, radius=2, shapeName=dyn_mochi02_capsule, center=<4 0 0>, rotation=<rot 90 -104.1513 -104.1513>}]
 dyn_mochi03 [{shapeType=Capsule, length=2, radius=2, shapeName=dyn_mochi03_capsule, center=<4 0 0>, rotation=<rot 90 53.731 53.731>}]
 dyn_mochi04 [{shapeType=Capsule, length=2, radius=2, shapeName=dyn_mochi04_capsule, center=<4 0 0>, rotation=<rot 90 -133.6167 -133.6167>}]
 dyn_mochi25 [{shapeType=Capsule, length=2, radius=2, shapeName=dyn_mochi25_capsule, center=<4 0 0>, rotation=<rot 90 -35.4012 -35.4012>}]
 dyn_mochi26 [{shapeType=Capsule, length=2, radius=2, shapeName=dyn_mochi26_capsule, center=<4 0 0>, rotation=<rot 90 -109.4919 -109.4919>}]
 dyn_mochi27 [{shapeType=Capsule, length=2, radius=2, shapeName=dyn_mochi27_capsule, center=<4 0 0>, rotation=<rot 90 93.7438 93.7438>}]
}
constraints [
 {bone1Name=dyn_mochi01, bone2Name=base_mochi01, swing1LimitDegrees=30, swing2LimitDegrees=30, twistLimitDegrees=1}
 {bone1Name=dyn_mochi02, bone2Name=base_mochi02, swing1LimitDegrees=40, swing2LimitDegrees=30, twistLimitDegrees=1}
 {bone1Name=dyn_mochi03, bone2Name=base_mochi03, swing1LimitDegrees=30, swing2LimitDegrees=30, twistLimitDegrees=1}
 {bone1Name=dyn_mochi04, bone2Name=base_mochi04, swing1LimitDegrees=30, swing2LimitDegrees=30, twistLimitDegrees=1}
 {bone1Name=dyn_mochi25, bone2Name=base_mochi25, swing1LimitDegrees=30, swing2LimitDegrees=30, twistLimitDegrees=1}
 {bone1Name=dyn_mochi26, bone2Name=base_mochi26, swing1LimitDegrees=30, swing2LimitDegrees=30, twistLimitDegrees=1}
 {bone1Name=dyn_mochi27, bone2Name=base_mochi27, swing1LimitDegrees=30, swing2LimitDegrees=30, twistLimitDegrees=1}
 {bone1Name=dyn_base, bone2Name=Melee_TwoHanded_Handle, swing1LimitDegrees=10, swing2LimitDegrees=10, twistLimitDegrees=1}
 {bone1Name=base_mochi01, bone2Name=dyn_base, swing1LimitDegrees=10, swing2LimitDegrees=10, twistLimitDegrees=1}
 {bone1Name=base_mochi02, bone2Name=dyn_base, swing1LimitDegrees=10, swing2LimitDegrees=10, twistLimitDegrees=1}
 {bone1Name=base_mochi03, bone2Name=dyn_base, swing1LimitDegrees=10, swing2LimitDegrees=10, twistLimitDegrees=1}
 {bone1Name=base_mochi04, bone2Name=dyn_base, swing1LimitDegrees=10, swing2LimitDegrees=10, twistLimitDegrees=1}
 {bone1Name=base_mochi26, bone2Name=Melee_TwoHanded_Handle, swing1LimitDegrees=10, swing2LimitDegrees=10, twistLimitDegrees=1}
 {bone1Name=base_mochi25, bone2Name=Melee_TwoHanded_Handle, swing1LimitDegrees=10, swing2LimitDegrees=10, twistLimitDegrees=1}
 {bone1Name=base_mochi27, bone2Name=Melee_TwoHanded_Handle, swing1LimitDegrees=10, swing2LimitDegrees=10, twistLimitDegrees=1}
 {
  bone1Name=Melee_TwoHanded_Handle
  bone2Name=b_AR_Root
  swing1LimitDegrees=45
  swing1Motion=Free
  swing2LimitDegrees=45
  swing2Motion=Free
  twistLimitDegrees=45
  twistMotion=Free
 }
]
